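; Parent Blueprint for ice cream truck quest props, playing an effect when the local player's truck-visit objective stops being valid.
; On non-authority only, BeginPlay waits through a workaround delay, then binds to the quest manager's OnQuestsUpdated if IsObjectiveValid passes.
; QuestsUpdated rechecks the first local controller's objective; once invalid, it activates the BullsEye_Land component and unbinds.
; DoesPlayerHaveThisQuest is defined but not called here.
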(event EventBeginPlay
  (bind _returnvalue (Fortnite|GetLocalFortPlayerControllers true))
  (bind _outputdelegate (EventDispatchers|CreateEvent :!event Custom|QuestsUpdated))
  (bind _output (Utilities|Array|Get(acopy) _returnvalue))
  ;[1 1 1] BEGIN PLAY; check if quest/objective valid
  (Networking|SwitchHasAuthority
    (:Remote
      ;;[1 0.5445 0.1188] Workaround - needs to be something better for ship
      (Utilities|FlowControl|Delay 5.0)
      (bind _objectivevalid (CallFunction|IsObjectiveValid))
      (if _objectivevalid
        ;[1 1 1] Bind Event to Truck Visited
        (bind _asathena_player_controller (Utilities|Casting|CastToAthena_PlayerController _output))
        (Quests|BindEventtoOnQuestsUpdated (Quests|GetQuestManager _asathena_player_controller "Invalid") _outputdelegate)
        (else
          ;[1 1 1] Unbind Event to Truck Visited
          (bind _asathena_player_controller_2 (Utilities|Casting|CastToAthena_PlayerController _output))
          (Quests|UnbindEventfromOnQuestsUpdated (Quests|GetQuestManager _asathena_player_controller_2 "Invalid") _outputdelegate))))))

(event Custom|QuestsUpdated
  ;[1 1 1] TruckVisited, fire FX and Sound
  (bind _objectivevalid (CallFunction|IsObjectiveValid))
  (if (not _objectivevalid)
    (Components|Activation|Activate (Variables|Default|GetP_Athena_BullsEye_Land) false)
    ;[1 1 1] Unbind Event to Truck Visited
    (bind _asathena_player_controller (Utilities|Casting|CastToAthena_PlayerController (Utilities|Array|Get(acopy) (Fortnite|GetLocalFortPlayerControllers true))))
    (Quests|UnbindEventfromOnQuestsUpdated (Quests|GetQuestManager _asathena_player_controller "Invalid") (EventDispatchers|CreateEvent :!event Custom|QuestsUpdated))))

(fn DoesPlayerHaveThisQuest
  (return (Utilities|IsValid (Quests|GetQuestwithDefinition (Quests|GetQuestManager (Utilities|Array|Get(acopy) (Fortnite|GetLocalFortPlayerControllers true)) "Invalid") (Variables|QuestStuff|GetQuestItem)) :!class /Script/Engine.KismetSystemLibrary)))

(fn IsObjectiveValid
  (bind _questvalid (Quest|DoesPlayerHaveQuest (Utilities|Array|Get(acopy) (Fortnite|GetLocalFortPlayerControllers true)) (Variables|QuestStuff|GetQuestItem) (Variables|QuestStuff|GetQuestObjectiveBackendName)))
  (return _questvalid))

(fn ConstructionScript)
